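; Quest vault scanner actor that plays a rotating scan effect when its state changes for a player.
; BeginPlay derives rotation rate and one-revolution duration from ScanningSpeed.
; Enabling unhides and resets the post-interaction mesh, starting RotatingMovement and the Niagara effect; duration expiry or cosmetic exit stops rotation, then the effect after a short delay.
; Construction script applies pre/post-interaction material overrides, including a preview toggle.
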(event EventBeginPlay
  (bind _returnvalue 3.6e+02)
  (bind _scanningspeed (Variables|Default|GetScanningSpeed))
  (bind _returnvalue_2 (* _returnvalue _scanningspeed))
  (|CallParentFunction)
  (Variables|Default|SetCachedInitialRotation (Variables|/Script/Engine.SceneComponent|GetRelativeRotation (Variables|Default|GetSM_Quest_PostInteraction_Extra)))
  (Components|Movement|SetUpdatedComponent (Variables|Default|GetRotatingMovement) (Variables|Default|GetSM_Quest_PostInteraction_Extra))
  (Variables|/Script/Engine.RotatingMovementComponent|SetRotationRate (Math|Rotator|MakeRotator 0.0 0.0 _returnvalue_2) (Variables|Default|GetRotatingMovement))
  (Variables|Default|SetScanningDurationTime (/ 1.0 (/ _returnvalue_2 3.6e+02))))

(event EventChangeStateForPlayer
  (params (FPC :type Fort%20Player%20Controller%20Object%20Reference))
  (|CallParentFunction FPC)
  (CallFunction|EnableScanner))

(event EventOnExitCosmetic
  (params (PlayerController :type Fort%20Player%20Controller%20Object%20Reference) (PlayerPawn :type Fort%20Player%20Pawn%20Athena%20Object%20Reference))
  (|CallParentFunction PlayerController PlayerPawn)
  (CallFunction|DisableScanner))

(event Custom|EnableScanner
  (if (not (Variables|Default|GetIsScanning))
    (Variables|Default|SetIsScanning true)
    (Development|SetHiddeninGame (Variables|Default|GetSM_Quest_PostInteraction_Extra) false true)
    ;[1 1 1] Place Any Initialization Code Here
    (Components|Activation|Activate (Variables|Default|GetNS_Scanner) false)
    (Transformation|SetRelativeRotation (Variables|Default|GetSM_Quest_PostInteraction_Extra) (Variables|Default|GetCachedInitialRotation) false false) ; Reset rotation and start animating
    (Components|Activation|Activate (Variables|Default|GetRotatingMovement) true)
    (Utilities|FlowControl|Delay (Variables|Default|GetScanningDurationTime))
    (CallFunction|DisableScanner)))

(event Custom|DisableScanner
  (if (Variables|Default|GetIsScanning)
    (Variables|Default|SetIsScanning false)
    (Components|Activation|Deactivate (Variables|Default|GetRotatingMovement))
    ;[1 1 1] Disable Code Goes Here
    (Utilities|FlowControl|Delay 0.2)
    (Components|Activation|Deactivate (Variables|Default|GetNS_Scanner))))

(fn PostInteractionVisuals
  (params (PlayerController :type Fort%20Player%20Controller%20Object%20Reference))
  (|CallParentFunction PlayerController))

(fn ConstructionScript
  (bind _preinteraction_mis_override (Variables|Design|PreInteraction|GetPreInteraction_MIS_Override))
  (bind _postinteraction_mis_override (Variables|Design|PostInteraction|GetPostInteraction_MIS_Override))
  (|CallParentFunction)
  (Utilities|FlowControl|Sequence
    (:then_1
      ;[1 1 1] Material Swaps
      (Utilities|IsValid _preinteraction_mis_override
        (:Is%20Valid
          (LocalVariables|SetStoredOriginalPreInteractionExtraMaterial (Utilities|Array|Get(acopy) (Rendering|Material|GetMaterials (Variables|Default|GetSM_Quest_PreInteraction_Extra))))
          (if (not (Variables|Design|PostInteraction|GetbPreviewPostInteraction))
            ;;[1 1 1] If PreInteraction has swapped material
            (Rendering|Material|SetMaterial (Variables|Default|GetSM_Quest_PreInteraction_Extra) 0 _preinteraction_mis_override)))
        (:Is%20Not%20Valid
          (if (Variables|Design|PostInteraction|GetbPreviewPostInteraction)
            _)))
      ;;[1 1 1] If Previewing the Post Interaction State
      (Utilities|IsValid _postinteraction_mis_override
        (:Is%20Valid
          (Utilities|IsValid (Variables|Default|GetSM_Quest_PostInteraction_Extra)
            (:Is%20Valid
              (Rendering|Material|SetMaterial (Variables|Default|GetSM_Quest_PostInteraction_Extra) 0 _postinteraction_mis_override))
            (:Is%20Not%20Valid
              (Rendering|Material|SetMaterial (Variables|Default|GetSM_Quest_PreInteraction_Extra) 0 _postinteraction_mis_override))))
        (:Is%20Not%20Valid
          (Rendering|Material|SetMaterial (Variables|Default|GetSM_Quest_PreInteraction_Extra) 0 (LocalVariables|GetStoredOriginalPreInteractionExtraMaterial)))))))
